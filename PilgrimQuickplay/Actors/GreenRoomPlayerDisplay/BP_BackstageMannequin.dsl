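; BP_BackstageMannequin is a backstage mannequin actor for the Quickplay green room player display, optimized for static posed animations.
; It overrides SetAnimationFromLoadedAnim to run the parent logic, then force-tick the skeletal mesh animation once and disable further skeleton updates.
; This saves animation cost, relying on all current mannequin animations being single poses.
; BeginPlay, Tick and overlap events are empty, and the construction script only calls its parent.
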(event EventBeginPlay)

(event Collision|EventActorBeginOverlap
  (params (OtherActor :type Actor%20Object%20Reference)))

(event EventTick
  (params (DeltaSeconds :type Float%20%28single-precision%29)))

(event EventSetAnimationFromLoadedAnim
  (|CallParentFunction)
  (bind _skeletalmeshcomponent (Variables|SkeletalMeshActor|GetSkeletalMeshComponent)
    (:then
      (Animation|Utilities|ForceTickAnimation _skeletalmeshcomponent)
      (Optimizations|SetNoSkeletonUpdate true) ; Currently, all backstage mannequin animations are single poses. We don't need to actually tick the animations after setting it up and ticking once
    )
    (:else)))

(fn ConstructionScript
  (|CallParentFunction))
