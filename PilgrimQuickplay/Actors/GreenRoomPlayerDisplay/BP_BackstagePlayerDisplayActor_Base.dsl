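; Base actor for a Pilgrim Quickplay backstage display that tracks one player by playspace player-state index.
; BeginPlay waits for playlist data and the Pilgrim event router, then runs empty server or client init hooks.
; On the server, playspace user-added events cache the indexed player state in replicated MyPlayerState, and removal of that player clears it.
; Its OnRep forwards to an empty OnPlayerStateReplicated function, a workaround for child Blueprint overrides.
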(event EventBeginPlay
  (bind _returnvalue (Class|SparksMusicPlayspace|GetSparksMusicPlayspaceOfClass "/Script/PilgrimQuickplayRuntime.PilgrimQuickplayPlayspace"))
  ;[0.625 1 1] Make sure this poster is looking for a valid player state
  (if (> (Variables|Default|GetPlayspacePlayerStateIndex) -1)
    ;[0.625 1 1] Wait for the event router
    (Class|FortAsyncActionCurrentPlaylistReady|PlaylistDataReadyAsync
      (:OnReady
        (bind _eventrouter (Pilgrim|PilgrimWaitForEventRouter 0)
          (:OnEventRouterAvailable
            (Variables|Default|SetEvent%20Router _eventrouter)
            (Networking|SwitchHasAuthority
              (:Authority
                ;[0.2118 0.4159 1] Server looks for a valid playspace
                (Utilities|IsValid _returnvalue
                  (:Is%20Valid
                    ;[0.2118 0.4159 1] Bind to user added and removed events from the playspace
                    (Playspace|BindEventtoNotifyPlayspaceUserAdded _returnvalue (EventDispatchers|CreateEvent :!event Custom|OnPlayspaceUserAdded))
                    (Playspace|BindEventtoNotifyPlayspaceUserRemoved _returnvalue (EventDispatchers|CreateEvent :!event Custom|OnPlayspaceUserRemoved))
                    (CallFunction|InitializeonServer))
                  (:Is%20Not%20Valid
                    (Development|PrintString "Invalid playspace"))))
              (:Remote
                (CallFunction|InitializeonClient)))))))
    (else
      (Development|PrintString "Player info actor does not have a valid playerstate index set"))))

(event Custom|TrySetPlayerstateFromPlayspace
  (bind _returnvalue (Class|SparksMusicPlayspace|GetSparksMusicPlayspaceOfClass "/Script/PilgrimQuickplayRuntime.PilgrimQuickplayPlayspace"))
  (bind _returnvalue_2 (Class|SparksMusicPlayspace|GetAllPlayerStates _returnvalue))
  (bind _playspaceplayerstateindex (Variables|Default|GetPlayspacePlayerStateIndex))
  ;[0.625 1 1] If the index this display is tracking valid in the playspace's list of playerstates, cache that state (or clear it) in the replicated variable
  (if (Utilities|Array|IsValidIndex _returnvalue_2 _playspaceplayerstateindex)
    (Variables|Default|SetMyPlayerState (Utilities|Array|Get(acopy) _returnvalue_2 _playspaceplayerstateindex))
    (else
      (Variables|Default|SetMyPlayerState 0))))

(event Custom|OnPlayspaceUserAdded
  (params (AddedUser :type Playspace%20User%20Structure%20%28by%20ref%29))
  ;[0.625 1 1] If a user was added to the playspace and this actor doesn't have a playerstate associated, try to add it
  (CallFunction|TrySetPlayerstatefromPlayspace))

(event Custom|OnPlayspaceUserRemoved
  (params (RemovedUser :type Playspace%20User%20Structure%20%28by%20ref%29))
  (bind (_userid _leafplayspace _playerstatecached) (Utilities|Struct|BreakPlayspaceUser RemovedUser))
  ;[0.625 1 1] If a user is removed from the playspace and it's the same playerstate this display is tracking, clear the state and hide the widget
  (if (== _playerstatecached (Variables|Default|GetMyPlayerState))
    (Variables|Default|SetMyPlayerState 0)))

(event Custom|InitializeOnServer)

(event Custom|InitializeOnClient)

(fn OnPlayerStateReplicated)

(fn OnRep_MyPlayerState
  ;[0.625 1 1] When the player state is replicated, react. Only doing this with a seperate function because I couldn't find the OnRep to override in the child blueprints.
  (CallFunction|OnPlayerStateReplicated))

(fn ConstructionScript)
