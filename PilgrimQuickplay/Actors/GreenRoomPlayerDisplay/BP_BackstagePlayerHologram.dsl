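; Backstage hologram actor for Pilgrim Quickplay that displays a player as a tinted, animated mannequin.
; On client init it spawns a transient child mannequin, hidden until the pawn is set and its customization loads.
; After a short delay working around early load notifications, it copies cosmetics, plays a random gender-matched idle and activates a Niagara effect.
; Player-state replication and readiness events drive hologram color, optionally using pin color, through dynamic materials and Niagara.
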(event EventInitializeOnClient
  (bind _transientchild (Variables|Default|GetTransientChild))
  ;[0.625 1 1] Spawn and attach the creative mannequin as a transient child actor
  (bind _spawnedcomponent (Fortnite|Utilities|SpawnTransientSceneComponent "/Script/Engine.ChildActorComponent" self (Variables|Default|GetBaseMesh) false))
  (ChildActorComponent|SetChildActorClass _spawnedcomponent "/PilgrimQuickplay/Actors/GreenRoomPlayerDisplay/BP_BackstageMannequin.BP_BackstageMannequin_C")
  (Fortnite|Utilities|RegisterComponent _spawnedcomponent)
  (Variables|Default|SetTransientChild _spawnedcomponent)
  (Transformation|AttachComponentToComponent (Variables|Default|GetTransientChild) (Variables|Default|GetBaseMesh) "None" "SnapToTarget" "SnapToTarget" "SnapToTarget" true)
  (Transformation|SetRelativeTransform (Variables|Default|GetTransientChild) (Math|Transform|MakeTransform "0.000000,0.000000,16.435925" "0.000000,-90.000000,0.000000" (Variables|Default|GetPawnScale)) false false)
  (Utilities|Array|Add (Variables|/Script/Engine.ActorComponent|GetComponentTags _transientchild) "IgnoredByBlueprintPreviewer")
  (bind _ascreative_mannequin (Utilities|Casting|CastToCreative_Mannequin (Variables|/Script/Engine.ChildActorComponent|GetChildActor _transientchild)))
  (bind _output_get (Variables|Default|SetMannequinActor _ascreative_mannequin))
  (Collision|SetActorEnableCollision _output_get false) ; Disable collision completely on the mannequin.
  (CallFunction|UpdateHologramIfPlayerPawnValid)
  ;[0.625 1 1] Listen for player readiness events. If this event was for this player, set the readiness of this player and change the colors accordingly.
  (bind _eventrouter (Pilgrim|PilgrimWaitForEventRouter 0)
    (:OnEventRouterAvailable
      (bind (_asynctaskproxy _payload) (GameplayEventRouter|StartListeningToEvent _eventrouter "/Script/PilgrimCoreRuntime.PilgrimEvent_PlayerState_ReadinessChanged")
        (:OnEventReceived
          (if (== (Variables|/Script/Engine.Pawn|GetPlayerState (Utilities|Struct|BreakPilgrimEventPlayerStateReadinessChanged _payload)) (Variables|Default|GetMyPlayerState))
            (Variables|Default|SetbPlayerIsReady (.bIsReady (Utilities|Struct|BreakPilgrimEventPlayerStateReadinessChanged _payload)))
            (Default|SetDeviceColors (CallFunction|GetHologramColor))))))))

(event Custom|Update%20Hologram%20if%20Player%20Pawn%20Valid
  (bind _fort_player_pawn (CallFunction|GetFortPlayerPawn))
  (Networking|CanExecuteCosmeticEvents
    (:True
      ;[0.625 1 1] If there's a valid player pawn, set it to the hologram when the customization is done loading
      (CallFunction|SetHologramVisiblity false)
      (if (Utilities|IsValid _fort_player_pawn :!class /Script/Engine.KismetSystemLibrary)
        (if (Customization|IsCharacterCustomizationLoadingCompleted _fort_player_pawn)
          (CallFunction|OnCharacterCustomizationLoaded 0)
          (else
            (Customization|BindEventtoOnFinishedCharacterCustomization _fort_player_pawn (EventDispatchers|CreateEvent :!event Custom|OnCharacterCustomizationLoaded))))
        (else
          (CallFunction|BindtoonPawnSet))))))

(event Custom|OnCharacterCustomizationLoaded
  (params (Pawn :type Fort%20Player%20Pawn%20Object%20Reference))
  (bind _ns_hologramdecal_a (Variables|Default|GetNS_HologramDecal_A))
  (bind _mannequinactor (Variables|Default|GetMannequinActor))
  ;[0.625 1 1] When character customization has loaded, update the hologram. Note that I don't like the delay here, but I get the event that loading has finished before loading has actually finished. Possible I could fix this using replication.
  (Utilities|FlowControl|Delay 0.1)
  (Class|CreativeMannequin|UpdateCosmeticfromPawn _mannequinactor (CallFunction|GetFortPlayerPawn))
  (Class|CreativeMannequin|LoadAndPlayIdleMontage _mannequinactor (CallFunction|GetHologramAnimation) 0.0)
  ;[0.1244 0.1948 1] Spawn a Specific Hologram Decal FX Projected on the Mannequins Bones
  (Components|Activation|Activate _ns_hologramdecal_a false)
  (Niagara|SetNiagaraSkeletalMeshComponent _ns_hologramdecal_a "Skeletal Mesh_UsrParam" (Variables|/Script/Engine.SkeletalMeshActor|GetSkeletalMeshComponent _mannequinactor))
  ;[0.625 1 1] Update color and visibility
  (Default|SetDeviceColors (CallFunction|GetHologramColor))
  (CallFunction|SetHologramVisiblity true))

(event Custom|BindToOnPawnSet
  (bind _myplayerstate (Variables|Default|GetMyPlayerState))
  ;[0.625 1 1] Attempt to bind to the 'onpawnset' event from the playerstate to set the pawn on the mannequin once it's been set
  (Utilities|IsValid _myplayerstate
    (:Is%20Valid
      (Events|BindEventtoOnPawnSet _myplayerstate (delegate Custom|OnPawnSet)))))

(event Custom|OnPawnSet
  (params (Player :type Player%20State%20Object%20Reference) (NewPawn :type Pawn%20Object%20Reference) (OldPawn :type Pawn%20Object%20Reference))
  ;[0.625 1 1] Attempt to bind to the 'onpawnset' event from the playerstate to set the pawn on the mannequin once it's been set
  (if (== (Variables|Default|GetMyPlayerState) Player)
    (bind _asfort_player_pawn (Utilities|Casting|CastToFortPlayerPawn NewPawn))
    (Variables|Default|SetFort%20Player%20Pawn _asfort_player_pawn)
    (CallFunction|UpdateHologramIfPlayerPawnValid)))

(fn AdjustPinColor
  (params (PinColor_R :type Float%20%28single-precision%29) (PinColor_G :type Float%20%28single-precision%29) (PinColor_B :type Float%20%28single-precision%29) (PinColor_A :type Float%20%28single-precision%29))
  (return (* (Utilities|Struct|MakeLinearColor (Math|Float|Max(Float) PinColor_R 1.0) PinColor_G PinColor_B 1.0) (Variables|Default|GetPinColorAdjustment))))

(fn CreateOrReattachMIDs
  (Material|ReattachOrCreateMID (Variables|Default|GetHologram_Cone) (Variables|Default|GetCone%20MID))
  (Material|ReattachOrCreateMID (Variables|Default|GetBaseMesh) (Variables|Default|GetBase%20MID))
  (Material|ReattachOrCreateMID (Variables|Default|GetBotHighlight) (Variables|Default|GetHighlight%20MID)))

(fn GetFortPlayerPawn
  (bind _myplayerstate (Variables|Default|GetMyPlayerState))
  ;[0.625 1 1] If the player state is invalid, set the fort player pawn to null and return it
  (Utilities|IsValid (Variables|Default|GetMyPlayerState)
    (:Is%20Valid
      ;[0.625 1 1] If the player state is valid and we already have a valid player pawn reference, return that
      (Variables|Default|GetFort%20Player%20Pawn
        (:then
          _)
        (:else
          ;[0.625 1 1] Otherwise, try to get a valid pawn from the player state. Otherwise, return null
          (Utilities|IsValid _myplayerstate
            (:Is%20Valid
              (bind _asfort_player_pawn (Utilities|Casting|CastToFortPlayerPawn (PlayerState|GetPawn _myplayerstate))
                (:then
                  (Variables|Default|SetFort%20Player%20Pawn _asfort_player_pawn))
                (:CastFailed
                  _))
              (return (Variables|Default|GetFort%20Player%20Pawn)))
            (:Is%20Not%20Valid
              _)))))
    (:Is%20Not%20Valid
      (bind _output_get (Variables|Default|SetFort%20Player%20Pawn 0))
      (return _output_get))))

(fn GetHologramAnimation
  (bind _fort_player_pawn (CallFunction|GetFortPlayerPawn))
  (bind _returnvalue (Customization|GetCharacterGender _fort_player_pawn))
  ;[0.625 1 1] Get a random animation from our animation map, based on character gender
  (bind _values (Utilities|Map|Values (Variables|Default|GetAnimation%20Map)))
  (bind _outitem (Utilities|Array|RandomArrayItem _values))
  (bind _outitem_genderanimation_10_f29c23c542b94e35effb98bb8f98fcc6 (Utilities|Struct|BreakCreativeAnimationByGender _outitem))
  (if (.ReturnValue (Utilities|Map|Find _outitem_genderanimation_10_f29c23c542b94e35effb98bb8f98fcc6 _returnvalue))
    (return (Utilities|Map|Find _outitem_genderanimation_10_f29c23c542b94e35effb98bb8f98fcc6 _returnvalue))
    (else
      (return (Utilities|Map|Find _outitem_genderanimation_10_f29c23c542b94e35effb98bb8f98fcc6 "Both")))))

(fn GetHologramColor
  (return (select (Variables|Default|GetbChangeColorWhenPlayerReady) (select (Variables|Default|GetbPlayerIsReady) (Variables|Default|GetHologramReadyColor) (Variables|Default|GetHologramNeutralColor)) (Variables|Default|GetHologramNeutralColor))))

(fn OnPlayerStateReplicated
  (bind _myplayerstate (Variables|Default|GetMyPlayerState))
  (bind _myplayerstate_2 (Variables|Default|GetMyPlayerState))
  (Utilities|FlowControl|Sequence
    (:then_0
      ;[0.625 1 1] Set the readiness of the player new player state
      (Utilities|IsValid _myplayerstate_2
        (:Is%20Valid
          (bind _returnvalue (Class|PilgrimQuickplayPlayerStateComponent|GetPilgrimQuickplayPlayerStateComponent _myplayerstate_2))
          (Utilities|IsValid _returnvalue
            (:Is%20Valid
              (Variables|Default|SetbPlayerIsReady (Class|PilgrimCorePlayerStateComponent|GetIsReady _returnvalue)))))))
    (:then_1
      ;[0.625 1 1] If we're using player pin color, get and set that now, or set it back to default if the player state isn't valid
      (if (and (Variables|Default|GetbUsePlayerPinColor) (Utilities|IsValid _myplayerstate :!class /Script/Engine.KismetSystemLibrary))
        (bind _asfort_player_state_athena (Utilities|Casting|CastToFortPlayerStateAthena _myplayerstate)
          (:then
            (bind _adjustedpincolor (CallFunction|AdjustPinColor (Class|FortPlayerStateZone|GetPinColor _asfort_player_state_athena)))
            (Variables|Default|SetHologramNeutralColor _adjustedpincolor))
          (:CastFailed
            _))
        (else
          _))
      (Variables|Default|SetHologramNeutralColor (Variables|Default|GetHologramDefaultColor)))
    (:then_2
      (CallFunction|UpdateHologramIfPlayerPawnValid))))

(fn SetHologramVisiblity
  (params (IsVisible :type Boolean))
  (bind _returnvalue (not IsVisible))
  (Development|SetHiddeninGame (Variables|Default|GetTransientChild) _returnvalue false)
  (bind _mannequinactor (Variables|Default|GetMannequinActor)
    (:then
      (Rendering|SetActorHiddenInGame _mannequinactor _returnvalue))
    (:else
      _))
  (Development|SetHiddeninGame (Variables|Default|GetHologram_Cone) _returnvalue false))

(fn SetDeviceColors
  (params (New%20Color :type Linear%20Color%20Structure))
  (bind _hologram_effect_strength (Variables|Default|GetHologram%20Effect%20Strength))
  (bind _returnvalue (/ _hologram_effect_strength 1e+02))
  (bind _local_color (LocalVariables|Getlocal%20Color))
  (bind _returnvalue_2 (Math|Color|LinearColor*Float _local_color _returnvalue))
  (bind _highlight_mid (Variables|Default|GetHighlight%20MID))
  ;[1 0.5253 0 0.2] Updating Colors of Base, Hologram Effect, Cone and UI
  (LocalVariables|Setlocal%20Color (Math|Color|LinearColor*Float New%20Color (Variables|Default|GetHighlightStrengthBaseMultiplier)))
  (Default|CreateorReattachMIDs)
  (Networking|CanExecuteCosmeticEvents
    (:True
      (Rendering|Material|SetVectorParameterValue _highlight_mid "FresnelColor" _returnvalue_2 :!class /Script/Engine.MaterialInstanceDynamic)
      (Rendering|Material|SetVectorParameterValue _highlight_mid "EdgeColor" _returnvalue_2 :!class /Script/Engine.MaterialInstanceDynamic)
      (Rendering|Material|SetVectorParameterValue (Variables|Default|GetCone%20MID) "Cone Color" _returnvalue_2 :!class /Script/Engine.MaterialInstanceDynamic)
      (Rendering|Material|SetVectorParameterValue (Variables|Default|GetBase%20MID) "Emissive" (LocalVariables|Getlocal%20Color) :!class /Script/Engine.MaterialInstanceDynamic)
      (Niagara|SetNiagaraVariableByString(LinearColor) (Variables|Default|GetNS_HologramDecal_A) "ColorBase" (LocalVariables|Getlocal%20Color)))))

(fn ConstructionScript)
