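; Backstage solo player display actor for the Pilgrim Quickplay green room, showing one player's readiness on a screen mesh.
; The server requests the playspace player state matching its index; replication triggers a cosmetic-only display update, whose override here is empty.
; Clients listen for readiness events through the gameplay event router, swapping ready or idle screen materials for the matching player state.
; EndPlay unregisters stored listeners, and HideWidget hides the player widget.
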(event EventEndPlay
  (params (EndPlayReason :type EEndPlayReason%20Enum :default "Destroyed"))
  (|CallParentFunction EndPlayReason)
  (for _array_element (Variables|Default|GetRouterEvents)
    (GameplayEventRouter|StopListeningtoEvent _array_element))
  (Utilities|Array|Clear (Variables|Default|GetRouterEvents)))

(event EventInitializeOnServer
  ;[0.2118 0.4159 1] Now try to get a playerstate from the playspace that matches our index
  (CallFunction|TrySetPlayerstatefromPlayspace))

(event EventInitializeOnClient
  ;[0.3456 1 0.287] Client starts listening for relevant events
  (CallFunction|SetupClientEventListeners))

(event Custom|SetupClientEventListeners
  (Utilities|FlowControl|Sequence
    (:then_2
      ;[1 0.8518 0.662] Individual player readiness events
      (bind (_asynctaskproxy _payload) (GameplayEventRouter|StartListeningToEvent (Variables|Default|GetEvent%20Router) "/Script/PilgrimCoreRuntime.PilgrimEvent_PlayerState_ReadinessChanged")
        (:then
          (Utilities|Array|Add (Variables|Default|GetRouterEvents) _asynctaskproxy))
        (:OnEventReceived
          (bind (_payload_forplayer _payload_bisready) (Utilities|Struct|BreakPilgrimEventPlayerStateReadinessChanged _payload))
          (bind _playerstatematch (CallFunction|CheckforMatchingPlayerState _payload_forplayer))
          (if _playerstatematch
            (Rendering|Material|SetMaterial (Variables|Default|GetScreenMesh) 1 (select _payload_bisready (Variables|Default|GetReadyScreenMaterial) (Variables|Default|GetIdleScreenMaterial)))))))))

(fn CheckForMatchingPlayerState
  (params (FortPlayerPawn :type Fort%20Player%20Pawn%20Object%20Reference))
  (return (== (Variables|/Script/Engine.Pawn|GetPlayerState FortPlayerPawn) (Variables|Default|GetMyPlayerState))))

(fn HideWidget
  (Class|WBPGreenRoomPlayerDisplay|HidePlayerDisplay (Variables|Default|GetPlayerDisplayWidget))
  (Fortnite|FortShippingLog self "Hiding the player display"))

(fn OnPlayerStateReplicated
  ;[0.625 1 1] If we're on the client, and we have a valid player state, set it to this display. Otherwise, hide the display.
  (Networking|CanExecuteCosmeticEvents
    (:True
      (CallFunction|UpdateDisplayifPlayerStateValid))))

(fn UpdateDisplayIfPlayerStateValid)

(fn ConstructionScript)
