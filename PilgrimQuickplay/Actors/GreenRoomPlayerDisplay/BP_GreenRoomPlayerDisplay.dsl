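; Green room player display actor for Pilgrim Quickplay that exposes player states and replicated timer state from the playspace's PlayerManager component.
; On BeginPlay it waits for the Pilgrim event router, a temporary stand-in for a playspace wait node, then looks up and caches that component.
; Critical shipping logs flag missing references.
; The timer getter validates the cached component, while the player states getter reads it unchecked.
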(event EventBeginPlay
  (bind _returnvalue (Class|SparksMusicPlayspace|GetSparksMusicPlayspaceOfClass "/Script/PilgrimQuickplayRuntime.PilgrimQuickplayPlayspace"))
  (bind _returnvalue_2 (Actor|GetComponentbyClass _returnvalue "/PilgrimQuickplay/Playspaces/BP_PilgrimQuickplay_PlayerManager_PlayspaceComponent.BP_PilgrimQuickplay_PlayerManager_PlayspaceComponent_C"))
  ;[1 0 0.04316] TEMP: Until we get a WaitForPilgrimPlayspace Node
  (Pilgrim|PilgrimWaitForEventRouter 0
    (:OnEventRouterAvailable
      (Utilities|IsValid _returnvalue
        (:Is%20Valid
          (Utilities|IsValid _returnvalue_2
            (:Is%20Valid
              _)
            (:Is%20Not%20Valid
              (Fortnite|FortShippingLog 0 "PILGRIM CRITICAL ERROR!  BP_GreeenRoomPlayerDisplay could not find valid BP_PilgrimQuickplay_PlayerManager_PlayspaceComponent in Playspace." true)
              (Utilities|FlowControl|DelayUntilNextTick)))
          (Variables|Default|SetPilgrimQuickplay_PlayerManager_PlayspaceComponent _returnvalue_2))
        (:Is%20Not%20Valid
          _))
      (Fortnite|FortShippingLog 0 "PILGRIM CRITICAL ERROR!  BP_GreeenRoomPlayerDisplay could not find valid Playspace Component." true))))

(fn GetDisplayedPlayerStates
  (return (Variables|/PilgrimQuickplay/Playspaces/BP_PilgrimQuickplay_PlayerManager_PlayspaceComponent.BP_PilgrimQuickplay_PlayerManager_PlayspaceComponent_C|GetMyPlayerStates (Variables|Default|GetPilgrimQuickplay_PlayerManager_PlayspaceComponent))))

(fn GetPilgrimTimerStateReplicated
  (bind _pilgrimquickplay_playermanager_playspacecomponent (Variables|Default|GetPilgrimQuickplay_PlayerManager_PlayspaceComponent))
  (Utilities|IsValid _pilgrimquickplay_playermanager_playspacecomponent
    (:Is%20Valid
      (return (Variables|/PilgrimQuickplay/Playspaces/BP_PilgrimQuickplay_PlayerManager_PlayspaceComponent.BP_PilgrimQuickplay_PlayerManager_PlayspaceComponent_C|GetPilgrimTimerState_Replicated _pilgrimquickplay_playermanager_playspacecomponent)))
    (:Is%20Not%20Valid
      (Fortnite|FortShippingLog 0 "PILGRIM CRITICAL ERROR!  PilgrimQuickplay_PlayerManager_PlayspaceComponent was not ser invalid when 'Get Pilgrim Timer State Replicated' was called" true))))

(fn ConstructionScript)
