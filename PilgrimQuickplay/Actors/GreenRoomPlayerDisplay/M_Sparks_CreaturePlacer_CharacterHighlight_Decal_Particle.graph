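# Translucent unlit Niagara sprite material that draws a screen-space character highlight, apparently for Sparks creature placers in the green room.
# Panning scanlines plus a deferred-only scene-normal fresnel tint the particle color, blended toward solid color at edges found by four diagonal scene-depth taps.
# A rounded-square UV mask and depth and distance fades bound it.
# World position offset camera-aligns and offsets the sprite; the forward path blanks and collapses it.
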Material /PilgrimQuickplay/Actors/GreenRoomPlayerDisplay/M_Sparks_CreaturePlacer_CharacterHighlight_Decal_Particle
graph {
 properties {
  bEnableResponsiveAA=true
  bUsedWithNiagaraSprites=true
  blendMode=BLEND_Translucent
  pixelDepthOffsetMode=PDOM_Legacy
  shadingModel=MSM_Unlit
  twoSided=true
  usedShadingModels=Unlit
 }
 nodes [
  n1 = ParticleColor()
  Add(A=@n1.RGB) {constB=-0.15}
  n2 = Panner(Coordinate=ScreenPosition().ViewportUV) {bFractionalPart=true, speedY=0.0068}
  n3 = Frac(None=Multiply(
    A=MaterialFunctionCall(Float2=@n2) {materialFunction=&BreakOutFloat2Components}.G
    B=ScalarParameter() {defaultValue=370, parameterName=ScanlineRepeat}
   ))
  n4 = Constant() {r=0.5}
  n5 = ScalarParameter() {defaultValue=0.35, parameterName=ScanlineRadius}
  n6 = ScalarParameter() {defaultValue=0.5, parameterName=ScanlineHardness}
  n7 = StaticSwitchParameter(
   True=Multiply(A=ScalarParameter() {defaultValue=0.04, parameterName=ScanlineOpacity}, B=SphereMask(A=@n3, B=@n4, Radius=@n5, Hardness=@n6))
   False=Constant()
  ) {defaultValue=true, parameterName=UseScanlines}
  n8 = MaterialFunctionCall(
   Exponent=ScalarParameter() {defaultValue=4, parameterName=FresnelExponent}
   Vertex / Pixel Normal=SceneTexture() {bFiltered=true, sceneTextureId=PPI_WorldNormal}.Color
  ) {materialFunction=&MF_Fresnel_FX}
  n9 = Multiply(A=@n1.RGB, B=ShadingPathSwitch(Default=@n7, Deferred=Add(A=@n7, B=@n8.Fresnel Inverted), Mobile=@n7) {desc=Migrated from FeatureLevelSwitch})
  n10 = ScreenPosition()
  n11 = Multiply(A=ScalarParameter() {defaultValue=1.5, parameterName=TexelScale}, B=SceneTexelSize())
  n12 = Multiply(A=@n11, B=Constant2Vector() {g=-1, r=-1})
  n13 = SceneDepth()
  n14 = Subtract(A=SceneDepth(UVs=Add(A=@n10.ViewportUV, B=@n12)).R, B=@n13.R)
  n15 = Multiply(A=@n11, B=Constant2Vector() {g=-1, r=1})
  n16 = Subtract(A=SceneDepth(UVs=Add(A=@n10.ViewportUV, B=@n15)).R, B=@n13.R)
  n17 = Constant2Vector() {g=1, r=1}
  n18 = SceneDepth(UVs=Add(A=@n10.ViewportUV, B=Multiply(A=@n11, B=@n17)))
  n19 = Add(A=@n10.ViewportUV, B=Multiply(A=@n11, B=Constant2Vector() {g=1, r=-1}))
  n20 = Add(A=Add(A=Add(A=@n14, B=@n16), B=Subtract(A=@n18.R, B=@n13.R)), B=Subtract(A=SceneDepth(UVs=@n19).R, B=@n13.R))
  n21 = Saturate(None=Divide(A=Abs(None=@n20), B=ScalarParameter() {defaultValue=500, parameterName=EdgeDistanceDivisor}))
  n22 = ConstantBiasScale(None=TextureCoordinate()) {bias=-0.5, scale=2}
  n23 = Constant()
  n24 = SphereMask(A=Power(Base=Abs(None=@n22)) {constExponent=6}, B=@n23, Radius=Constant() {r=1}, Hardness=@n23)
  n25 = Multiply(
   A=Multiply(A=LinearInterpolate(A=@n9, B=@n1.RGB, Alpha=@n21), B=@n24)
   B=MaterialFunctionCall!(
    Distance Bias=ScalarParameter() {defaultValue=50, parameterName=Distance Bias}
    Distance Divisor=ScalarParameter() {defaultValue=70, parameterName=Distance Divisor}
    Scene Depth
    Pixel Depth
   ) {materialFunction=/Game/Effects/Fort_Effects/Materials/_Functions/MF_DistanceBehindTranslucency_01}.Inverted
  )
  n26 = ScalarParameter() {defaultValue=10, parameterName=DepthFadeDistance}
  n27 = ScalarParameter() {defaultValue=50, parameterName=NearCameraFadeDistance}
  n28 = ShadingPathSwitch(
   Default=@n25
   Deferred=Multiply(
    A=DepthFade(Opacity=@n25, FadeDistance=@n26)
    B=MaterialFunctionCall(Distance=@n27) {materialFunction=&Near_Camera_Fade}.Result
   )
   Mobile=@n25
  ) {desc=Migrated from FeatureLevelSwitch}
  n29 = Multiply(A=OneMinus(None=@n24), B=VectorParameter() {parameterName=DebugColor, defaultValue=<rgba 2 0 0 0>}.RGB)
  n30 = Multiply(
   A=StaticSwitchParameter(True=Add(A=@n28, B=@n29), False=@n28) {parameterName=DebugEdges}
   B=MaterialFunctionCall(
    Fade_Start_Distance=ScalarParameter() {defaultValue=7000, parameterName=Fade_Start_Distance}
    Fade_End_Distance=ScalarParameter() {defaultValue=1000, parameterName=Fade_End_Distance}
   ) {materialFunction=&MF_Distance_SyncedFade_FX}.Result Inverted
  )
  n31 = ShadingPathSwitch(Default=@n30, Deferred=@n30, Forward=Constant(), Mobile=@n30)
  MaterialFunctionCall(Emissive=Multiply(B=@n31) {constA=10}) {materialFunction=&MF_FX_EmissiveAdjustments}
  n32 = DepthFade() {fadeDistanceDefault=80}
  n33 = Multiply(B=@n31) {constA=1000}
  Multiply(A=Multiply(A=OneMinus(None=@n32), B=Saturate(None=@n33)), B=ParticleColor().A)
  n34 = WorldPosition()
  n35 = Constant3Vector() {constant=<rgba 0 1 0 1>}
  n36 = Constant3Vector() {constant=<rgba 0 0 -1 0>}
  n37 = Normalize(VectorInput=Subtract(A=CameraPositionWS(), B=ObjectPositionWS()))
  n38 = ScalarParameter() {defaultValue=60, parameterName=NearCameraScale}
  n39 = Add(
   A=Add(
    A=MaterialFunctionCall!(WorldPosition=@n34.XYZ, PivotLocation, WS Normals, Custom Object Basis 1=@n35, Custom Object Basis 2=@n36, Custom Object Basis 3) {materialFunction=/Engine/Functions/Engine_MaterialFunctions02/WorldPositionOffset/AlignMeshToTheCamera}.World Position Offset
    B=Multiply(A=@n37, B=@n38)
   )
   B=MaterialFunctionCall(
    Offset Amount=ScalarParameter() {defaultValue=65, desc="A Slight Camera Offset for Non Standard Emote Poses ", parameterName=Camera Offset}
    World Position=WorldPosition().XYZ
   ) {materialFunction=/Engine/Functions/Engine_MaterialFunctions02/WorldPositionOffset/CameraOffset}.Camera Offset
  ) {desc="NEW For Sparks!"}
  ShadingPathSwitch(
   Default=@n39
   Deferred=@n39
   Forward=Subtract(
    A=MaterialFunctionCall() {materialFunction=&ObjectPivotPoint}.Object Pivot Location
    B=WorldPosition().XYZ
   )
   Mobile=@n39
  )
  n40 = VectorParameter() {bUseCustomPrimitiveData=true, parameterName=FresnelColorCustomPrimitive0-3, defaultValue=<rgba 1.7474 0 2 0>}
  StaticSwitchParameter(True=@n40.RGB, False=Multiply(A=@n1.RGB) {constB=2}) {parameterName=FresnelColorUsesCustomPrimitiveData0-3}
  VectorParameter() {parameterName=FresnelColor, defaultValue=<rgba 0 1.5374 2 1>}
  VectorParameter() {parameterName=EdgeColor, defaultValue=<rgba 0 0.5555 5 1>}
 ]
}
