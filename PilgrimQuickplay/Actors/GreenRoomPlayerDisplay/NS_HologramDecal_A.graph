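# Niagara system for the green room player display that places hologram decal sprites on a character's skeletal mesh, tinted by User.ColorBase.
# A burst spawns 17 particles, partitioned so each samples its own bone or socket index every update, with array-driven position offsets.
# Particle lifetimes loop rather than expire, and a camera offset pulls the two sprite renderers' particles toward the view.
# Spawn rate and dynamic material parameter modules are disabled.
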NiagaraSystem /PilgrimQuickplay/Actors/GreenRoomPlayerDisplay/NS_HologramDecal_A
dependencies {
 usedDataInterfaces [&NiagaraDataInterfaceArrayFloat3]
 usedDynamicInputs [&MakeCustomFloatFromBool]
 usedModules [
  &self:SystemSpawnScript.SystemScriptSource.SystemScriptGraph.NiagaraNodeAssignment_0.SetVariables_#1
  &SystemState
 ]
}
emitters {Hologram_Decals {
  emitterUpdateScript [
   &EmitterState(
    Inactive Response=::NewEnumerator1
    Life Cycle Mode=::NewEnumerator1
    Loop Behavior=::NewEnumerator1
    Loop Duration Mode=::NewEnumerator1
    UseLoopDelay=false
   )
   &SpawnRate(SpawnRate=15, Use Spawn Probability=false) {enabled=false}
   &SpawnBurst_Instantaneous(Spawn Count=17, Use Spawn Probability=false, UseLoopCountLimit=false)
  ]
  particleSpawnScript [
   /SetVariables(Particles.Random=&RandomRangeFloat(Minimum=0, Maximum=1)) {enabled=false, isSetParameters=true}
   &PartitionParticles(Number Of Partitions=17)
   /Niagara/Modules/Spawn/Initialization/InitializeParticle(
    Color=<rgba 0.6414 0 1 1>
    Lifetime=&Lerp_Float(A=0.5, B=1, Alpha=$Particles.Random)
    Module.Write Color=true
    Module.Write Lifetime=true
    Module.Write Mass=true
    Module.Write Material Random=true
    Module.Write Position=false
    Module.Write Scale=false
    Module.Write SpriteRotation=false
    Module.Write SpriteSize=true
    Position=&SimulationPosition()
    Sprite Size=<75 75>
   ) {enabled=false}
   /Niagara/Modules/Spawn/Initialization/V2/InitializeParticle(
    Color=$User.ColorBase
    Lifetime=&Lerp_Float(A=0.5, B=1, Alpha=$Particles.Random)
    Mass Mode=::NewEnumerator1
    Sprite Size=<125 250>
    Sprite Size Mode=::NewEnumerator4
    UsePositionOffset=false
    Write Material Random=true
   ) {module=InitializeParticle001}
  ]
  particleUpdateScript [
   &ParticleState(
    Kill Particles When Lifetime Has Elapsed=false
    Let Infinitely Lived Particles Die When Emitter Deactivates=true
    Loop Particles Lifetime=true
   )
   &Color(Color=$User.ColorBase)
   &SkeletalMeshLocation(
    Bone / Socket Index=$Particles.PartitionParticles.Partition
    Bone Sampling Mode=::NewEnumerator4
    MaskSpawnGroup=false
    Orientation Sampling=::NewEnumerator1
    Sampled Position Offset=&SelectVectorFromArray(
     Direct Array Index=$Particles.PartitionParticles.Partition
     Vector Selection Array=DataInterface() {propertyValues {FloatData [<0 0 -5>, <0 0 -5>, <0 0 -5>, <0 0 0>, <0 0 0>, <0 0 0>, <0 0 0>]}}
    )
    Sampled Position Offset Coordinate Space:ENiagaraSkelMeshTransforms=::NewEnumerator1
    Skeletal Mesh=$User.Skeletal Mesh_UsrParam
   )
   &SolveForcesAndVelocity(Clamp Velocity=false, Limit Acceleration=false)
   &DynamicMaterialParameters(
    Index 0 Param 1=&Lerp_Float(A=0.2, B=1.1, Alpha=$Particles.Random)
    Index 0 Param 2=&Lerp_Float(A=1, B=2, Alpha=$Particles.Random)
    Index 0 Param 3=&Lerp_Float(A=0.25, B=0.95, Alpha=$Particles.Random)
    Index 0 Param 4=-1
    Module.Param0WriteEnabled=true
    Module.Param1WriteEnabled=true
    Module.Param2WriteEnabled=true
    Module.Param3WriteEnabled=true
   ) {enabled=false}
   &CameraOffset(Camera Offset Amount=35)
  ]
  rendererClasses [&NiagaraSpriteRendererProperties]
  renderers [
   {rendererClass=&NiagaraSpriteRendererProperties}
   {rendererClass=&NiagaraSpriteRendererProperties, rendererIndex=1}
  ]
 }}
summary {
 systemName=NS_HologramDecal_A
 emitters [{emitterName=Hologram_Decals, rendererClasses [&NiagaraSpriteRendererProperties]}]
 userVariables [
  User.ColorBase {type=LinearColor, description="", defaultValue {struct=LinearColor, value=<rgba 0 0.8725 2 1>}}
  User.Interpolate Intensity {type=NiagaraFloat, defaultValue=1, description=""}
  User.IsFrontEnd {type=NiagaraBool, defaultValue=true}
  User.IsLargeBody {type=NiagaraBool}
  User.Skeletal Mesh_UsrParam {
   type=NiagaraDataInterfaceSkeletalMesh
   defaultValue {
    struct=NiagaraExt_VariableValue_DataInterface
    value {dataInterface=&self:NiagaraDataInterfaceSkeletalMesh_0}
   }
  }
 ]
}
